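# M_Potion is a Default Lit material for the VigorKnackDill backpack cosmetic, built on material attributes.
# The 2023 character-parent PreFX function feeds MF_Potion, whose result goes into the matching PostFX function along with auxiliary data from PreFX.
# The potion-specific look lives inside MF_Potion, sandwiched between the shared character stages.
# It flags pixel animation and outputs a constant temporal responsiveness of 1, which affects how temporal history treats its pixels.
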Material /BRCosmetics/Accessories/FORT_Backpacks/Backpack_M_MED_VigorKnackDill/Materials/M_Potion
graph {
 properties {bHasPixelAnimation=true, bUseMaterialAttributes=true, pixelDepthOffsetMode=PDOM_Legacy, usedShadingModels=Default Lit}
 nodes [
  n1 = MaterialFunctionCall() {materialFunction=&MF_CharacterParent_2023_PreFX}
  MaterialFunctionCall(
   Material=MaterialFunctionCall!(In=@n1.Result) {materialFunction=&/MF_Potion}.Result
   M=@n1.M
   Pre Skinned Local Position=@n1.M
   Distance Blend=@n1.M
   Fresnel=@n1.M
  ) {materialFunction=&MF_CharacterParent_2023_PostFX}
  TemporalResponsivenessOutput!(None=Constant() {r=1})
 ]
}
