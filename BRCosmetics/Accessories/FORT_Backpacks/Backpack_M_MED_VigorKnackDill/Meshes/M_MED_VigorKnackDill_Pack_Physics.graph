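# Physics asset for the VigorKnackDill backpack cosmetic, defining simulation bodies for the pack and its attached head, arms, legs and flask.
# Capsules cover the pack and dynamic limb and flask bones, a large sphere covers the head, and small spheres mark base bones.
# Each dynamic part hangs from its base through an angle-limited constraint.
# Only the left arm base has a constraint to the pack, with all angular motion locked.
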PhysicsAsset /BRCosmetics/Accessories/FORT_Backpacks/Backpack_M_MED_VigorKnackDill/Meshes/M_MED_VigorKnackDill_Pack_Physics
bodies {
 base_arm_l [{shapeType=Sphere, radius=1, shapeName=base_arm_l_sphere}]
 base_arm_r [{shapeType=Sphere, radius=1, shapeName=base_arm_r_sphere}]
 base_flask [{shapeType=Sphere, radius=1, shapeName=base_flask_sphere}]
 base_leg_l [{shapeType=Sphere, radius=1, shapeName=base_leg_l_sphere}]
 base_leg_r [{shapeType=Sphere, radius=1, shapeName=base_leg_r_sphere}]
 dyn_arm_l_01 [{
   shapeType=Capsule
   length=5.2217
   radius=3.2907
   shapeName=dyn_arm_l_01_capsule
   center=<5.3361 0.9867 1.1126>
   rotation=<rot 0 87.2223 -59.2188>
  }]
 dyn_arm_r_01 [{
   shapeType=Capsule
   length=5.2217
   radius=3.2907
   shapeName=dyn_arm_r_01_capsule
   center=<5.3365 0.9866 -1.1122>
   rotation=<rot 0 -92.7735 -59.2256>
  }]
 dyn_flask_01 [{
   shapeType=Capsule
   length=3.046
   radius=1.9694
   shapeName=dyn_flask_01_capsule
   center=<4.1818 0.0608 0.0769>
   rotation=<rot 0 102.2957 -88.9745>
  }]
 dyn_leg_l_01 [{
   shapeType=Capsule
   length=4.7226
   radius=4.1888
   shapeName=dyn_leg_l_01_capsule
   center=<4.6392 -0.0381 -0.341>
   rotation=<rot 0 -91.8651 -88.5341>
  }]
 dyn_leg_r_01 [{
   shapeType=Capsule
   length=4.7226
   radius=4.1888
   shapeName=dyn_leg_r_01_capsule
   center=<4.6583 -0.0388 0.4046>
   rotation=<rot 0 -91.8651 -88.5341>
  }]
 head_base [{shapeType=Sphere, radius=1, shapeName=head_base_sphere}]
 head_dyn [{shapeType=Sphere, radius=11.3326, shapeName=head_dyn_sphere, center=<-10.6244 -0.9602 -0.2072>}]
 pack [{shapeType=Capsule, length=4.0342, radius=7.3395, center=<0.1063 -6.5142 -8.617>, rotation=<rot -0.9897 -90 -91.2643>}]
}
constraints [
 {bone1Name=dyn_leg_l_01, bone2Name=base_leg_l, swing1LimitDegrees=65, swing2LimitDegrees=20, twistLimitDegrees=10}
 {bone1Name=dyn_arm_r_01, bone2Name=base_arm_r, swing1LimitDegrees=4, swing2LimitDegrees=30, twistLimitDegrees=20}
 {bone1Name=dyn_leg_r_01, bone2Name=base_leg_r, swing1LimitDegrees=55, swing2LimitDegrees=20, twistLimitDegrees=10}
 {bone1Name=dyn_flask_01, bone2Name=base_flask, swing1LimitDegrees=15, swing2LimitDegrees=15, twistLimitDegrees=5}
 {
  bone1Name=base_arm_l
  bone2Name=pack
  swing1LimitDegrees=45
  swing1Motion=Locked
  swing2LimitDegrees=45
  swing2Motion=Locked
  twistLimitDegrees=45
  twistMotion=Locked
 }
 {bone1Name=dyn_arm_l_01, bone2Name=base_arm_l, swing1LimitDegrees=4, swing2LimitDegrees=30, twistLimitDegrees=20}
 {bone1Name=head_dyn, bone2Name=head_base, swing1LimitDegrees=10, swing2LimitDegrees=10, twistLimitDegrees=10}
]
