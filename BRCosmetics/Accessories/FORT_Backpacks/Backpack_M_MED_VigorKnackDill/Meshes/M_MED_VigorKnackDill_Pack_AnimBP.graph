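# Animation Blueprint for the VigorKnackDill cosmetic backpack, adding secondary physics motion and adjusting the pack's placement on the body.
# The input pose gets a constant additive pack-bone offset, a second offset weighted by Parashuting, then RigidBody simulation in pack-bone space.
# Each update derives Parashuting from spine_05's yaw delta against spine_01, clamped through a range map.
# An Anim Dynamics chain for head, limbs and flask bones is present but unconnected to the output.
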AnimBlueprint /BRCosmetics/Accessories/FORT_Backpacks/Backpack_M_MED_VigorKnackDill/Meshes/M_MED_VigorKnackDill_Pack_AnimBP
target_skeleton=&/M_MED_VigorKnackDill_Pack_Skeleton
graphs [
 AnimGraph {nodes [
   n1 = Anim Dynamics - Bone: head_dyn(Alpha=1, ExternalForce="0,0,0") {
    angularDampingOverride=0.9999
    angularSpringConstant=150
    bAngularSpring=true
    bOverrideAngularDamping=true
    bUseGravityOverride=true
    simulationSpace=RootRelative
    boundBone {boneName=head_dyn}
    componentAppliedLinearAccClamp=<1000 1000 1000>
    componentLinearAccScale=<1 0.6 0.4>
    componentLinearVelScale=<1 0.6 0.4>
    gravityOverride {z=-980}
    physicsBodyDefinitions [{
      boundBone {boneName=head_dyn}
      boxExtents=<10 10 10>
      constraintSetup {angularLimitsMax=<7 5 10>, angularLimitsMin=<-7 -5 -10>, angularTarget=<1 0 0>}
      localJointOffset=<1 0 0>
     }]
   }
   n2 = Anim Dynamics - Bone: dyn_leg_l_01(ComponentPose=@n1, Alpha=1, ExternalForce="0,0,0") {
    angularDampingOverride=0.999
    angularSpringConstant=150
    bAngularSpring=true
    bOverrideAngularDamping=true
    bUseGravityOverride=true
    simulationSpace=RootRelative
    boundBone {boneName=dyn_leg_l_01}
    componentAppliedLinearAccClamp=<3000 3000 3000>
    componentLinearAccScale=<1 0.6 0.4>
    componentLinearVelScale=<1 0.6 0.4>
    gravityOverride {z=-980}
    physicsBodyDefinitions [{
      boundBone {boneName=dyn_leg_l_01}
      boxExtents=<4 4 4>
      constraintSetup {angularLimitsMax=<0 10 10>, angularLimitsMin=<0 -10 -30>, angularTarget=<1 0 0>}
      localJointOffset=<0.1 0 0>
     }]
   }
   n3 = Anim Dynamics - Bone: dyn_leg_r_01(ComponentPose=@n2, Alpha=1, ExternalForce="0,0,0") {
    angularDampingOverride=0.999
    angularSpringConstant=150
    bAngularSpring=true
    bOverrideAngularDamping=true
    bUseGravityOverride=true
    simulationSpace=RootRelative
    boundBone {boneName=dyn_leg_r_01}
    componentAppliedLinearAccClamp=<3000 3000 3000>
    componentLinearAccScale=<1 0.6 0.4>
    componentLinearVelScale=<1 0.6 0.4>
    gravityOverride {z=-980}
    physicsBodyDefinitions [{
      boundBone {boneName=dyn_leg_r_01}
      boxExtents=<4 4 4>
      constraintSetup {angularLimitsMax=<0 10 10>, angularLimitsMin=<0 -10 -30>, angularTarget=<1 0 0>}
      localJointOffset=<0.1 0 0>
     }]
   }
   n4 = Anim Dynamics - Bone: dyn_arm_l_01(ComponentPose=@n3, Alpha=1, ExternalForce="0,0,0") {
    angularDampingOverride=0.999
    angularSpringConstant=150
    bAngularSpring=true
    bOverrideAngularDamping=true
    bUseGravityOverride=true
    simulationSpace=RootRelative
    boundBone {boneName=dyn_arm_l_01}
    componentAppliedLinearAccClamp=<3000 3000 3000>
    componentLinearAccScale=<1 0.6 0.4>
    componentLinearVelScale=<1 0.6 0.4>
    gravityOverride {z=-980}
    physicsBodyDefinitions [{
      boundBone {boneName=dyn_arm_l_01}
      boxExtents=<4 4 4>
      constraintSetup {angularLimitsMax=<0 30 0>, angularLimitsMin=<0 0 -20>, angularTarget=<1 0 0>}
      localJointOffset=<0.1 0.1 0>
     }]
   }
   n5 = Anim Dynamics - Bone: dyn_arm_r_01(ComponentPose=@n4, Alpha=1, ExternalForce="0,0,0") {
    angularDampingOverride=0.999
    angularSpringConstant=150
    bAngularSpring=true
    bOverrideAngularDamping=true
    bUseGravityOverride=true
    simulationSpace=RootRelative
    boundBone {boneName=dyn_arm_r_01}
    componentAppliedLinearAccClamp=<3000 3000 3000>
    componentLinearAccScale=<1 0.6 0.4>
    componentLinearVelScale=<1 0.6 0.4>
    gravityOverride {z=-980}
    physicsBodyDefinitions [{
      boundBone {boneName=dyn_arm_r_01}
      boxExtents=<4 4 4>
      constraintSetup {angularLimitsMax=<0 0 0>, angularLimitsMin=<0 -30 -20>, angularTarget=<1 0 0>}
      localJointOffset=<0.1 0.1 0>
     }]
   }
   n6 = Anim Dynamics - Bone: dyn_flask_02(ComponentPose=@n5, Alpha=1, ExternalForce="0,0,0") {
    angularDampingOverride=0.5
    angularSpringConstant=300
    bOverrideLinearDamping=true
    bUseGravityOverride=true
    linearDampingOverride=0.2
    linearSpringConstant=100
    simulationSpace=RootRelative
    boundBone {boneName=dyn_flask_02}
    componentAppliedLinearAccClamp=<1000 1000 1000>
    componentLinearAccScale {x=0.3}
    gravityOverride {z=-980}
    physicsBodyDefinitions [{
      boundBone {boneName=dyn_flask_02}
      boxExtents=<4 4 4>
      constraintSetup {angularLimitsMax=<0 0 0>, angularLimitsMin=<0 0 0>, angularTarget=<0 0 0>}
      localJointOffset=<0.1 0 0>
     }]
   }
   Anim Dynamics - Bone: dyn_flask_01(ComponentPose=@n6, Alpha=1, ExternalForce="0,0,0") {
    angularDampingOverride=0.999
    angularSpringConstant=300
    bAngularSpring=true
    bOverrideAngularDamping=true
    bUseGravityOverride=true
    simulationSpace=RootRelative
    boundBone {boneName=dyn_flask_01}
    componentAppliedLinearAccClamp=<1000 1000 1000>
    componentLinearAccScale=<0.5 0.4 0.6>
    componentLinearVelScale=<0.5 0.4 0.6>
    gravityOverride {z=-980}
    physicsBodyDefinitions [{
      boundBone {boneName=dyn_flask_01}
      boxExtents=<4 4 4>
      constraintSetup {angularLimitsMax=<2 7 0>, angularLimitsMin=<-2 -7 -10>, angularTarget=<0 0 0>}
      localJointOffset=<0.1 0 0>
     }]
   }
   n7 = "Transform (Modify) Bone - Bone: pack"(ComponentPose=Local To Component(LocalPose=Input Pose()), Translation="0,1,0", Rotation="0,0,10") {
    rotationMode=BMM_Additive
    scaleMode=BMM_Additive
    translationMode=BMM_Additive
    boneToModify {boneName=pack}
    rotation {roll=10}
    translation {y=1}
   }
   n8 = "Transform (Modify) Bone - Bone: pack"(ComponentPose=@n7, Alpha=Get Parashuting() -> (Parashuting:double), Translation="0,-10,4", Rotation="0,0,15") {
    rotationMode=BMM_Additive
    scaleMode=BMM_Additive
    translationMode=BMM_Additive
    boneToModify {boneName=pack}
    rotation {roll=15}
    translation {y=-10, z=4}
   }
   n9 = RigidBody(ComponentPose=@n8) {
    bClampLinearTranslationLimitToRefPose=true
    bOverrideWorldGravity=true
    overridePhysicsAsset=&/M_MED_VigorKnackDill_Pack_Physics
    simulationSpace=BaseBoneSpace
    baseBoneRef {boneName=pack}
    componentAppliedLinearAccClamp=<1000 1000 1000>
    overrideWorldGravity {z=-980}
    simSpaceSettings {dampingAlpha=0, maxAngularAcceleration=500, maxAngularVelocity=4, maxLinearAcceleration=1500, maxLinearVelocity=400, worldAlpha=0.7}
   }
   Output Pose(Result=Component To Local(ComponentPose=@n9)) {layerGroup=None}
  ]}
 EventGraph {nodes [
   Event BlueprintUpdateAnimation()
   n1 = GetDeltaTransformFromRefPose!(self=GetOwnerSkeletalMeshComponent(), BoneName=spine_05, BaseName=spine_01) -> (ReturnValue_Location:Vector, ReturnValue_Rotation:Rotator, ReturnValue_Scale:Vector)
   > Set Parashuting(Parashuting:double=MapRangeClamped(Value=BreakRotator(InRot=@n1.ReturnValue_Rotation).Yaw, InRangeA=-25, InRangeB=-65)) -> (Output_Get:double)
   TryGetPawnOwner()
  ]}
]
variables [Parashuting {type=number}]
